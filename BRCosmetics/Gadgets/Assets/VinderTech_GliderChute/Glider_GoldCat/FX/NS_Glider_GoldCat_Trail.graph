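# Niagara trail effect for the GoldCat glider, with separate tuning for gameplay and front-end display.
# Three ribbon emitters (variant-colored base, gold, black detail) anim-trail between socket pairs on the skeletal mesh, tinted by time of day.
# Gold glitter and sparkles spawn at selected sockets, alongside cylindrical speedlines; their gameplay spawn rates scale with User.ForwardDot.
# A system-level integrator drives gameplay opacity of ribbons and sparkles.
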NiagaraSystem /BRCosmetics/Gadgets/Assets/VinderTech_GliderChute/Glider_GoldCat/FX/NS_Glider_GoldCat_Trail
dependencies {
 usedDataInterfaces [
  &NiagaraDataInterfaceCurve
  &NiagaraDataInterfaceVector2DCurve
  &NiagaraDataInterfaceArrayFloat
  &NiagaraDataInterfaceArrayInt32
  &NiagaraDataInterfaceArrayColor
 ]
 usedDynamicInputs [&SelectLinearColorFromArray]
 usedModules [
  &self:SystemSpawnScript.SystemScriptSource.SystemScriptGraph.NiagaraNodeAssignment_1.SetVariables_#1
  &SystemState
  &FN_Integrator
  &self:SystemSpawnScript.SystemScriptSource.SystemScriptGraph.NiagaraNodeAssignment_0.SetVariables_#2
 ]
}
emitters {
 BaseDarkTrail {
  emitterSpawnScript [&FN_RibbonIDOffsetUtility()]
  emitterUpdateScript [
   &EmitterState(
    Inactive Response=::NewEnumerator1
    Life Cycle Mode=::NewEnumerator1
    Loop Behavior=::NewEnumerator1
    Loop Duration Mode=::NewEnumerator1
    UseLoopDelay=false
   )
   &FN_SpawnAnimTrail(Number of Anim Trails=4)
  ]
  particleSpawnScript [
   &FN_InitializeAnimTrailBoneIndecies()
   &FN_InitializeAnimTrailBoneIndecies(AnimTrail Selection=1, First Bone or Socket Index=2, Second Bone or Socket Index=3) {module=FN_InitializeAnimTrailBoneIndecies001}
   &FN_InitializeAnimTrailBoneIndecies(AnimTrail Selection=2, First Bone or Socket Index=6, Second Bone or Socket Index=7) {module=FN_InitializeAnimTrailBoneIndecies002}
   &FN_InitializeAnimTrailBoneIndecies(AnimTrail Selection=3, First Bone or Socket Index=8, Second Bone or Socket Index=9) {module=FN_InitializeAnimTrailBoneIndecies003}
   /SetVariables(Particles.AlphaFadeIn:NiagaraFloat=/Niagara/DynamicInputs/Math/Min_Float(
     A:NiagaraFloat=&Divide_Float(
      A=&Subtract_Float(A=$Emitter.Age, B=&FN_FrontEndFloatModifier(Gameplay=0, Front End=1.5))
      B=1
     )
     B:NiagaraFloat=1
    )) {isSetParameters=true, module=SetVariables_#3}
   &FN_InitializeAnimTrail(
    AnimTrailWidth=&FN_FrontEndFloatModifier(Gameplay=1, Front End=0.6)
    Color=<rgba 0 0 0 2>
    Lifetime=&MakeCustomFloatFromBool(Bool=$Particles.Locked, True Float=100, False Float=1)
    Position Offset=&FN_FrontEndVector3Modifier(Gameplay=<0 -12 0>, Front End=<0 -15 0>)
    SkeletalMesh=$System.Skeletal Mesh
   )
   &AddVelocity(Velocity=&Multiply_VectorByFloat(Vector=<0 400 0>, Float=&MakeCustomFloatFromBool(Bool=$User.bIsFrontEnd, True Float=1, False Float=1)))
  ]
  particleUpdateScript [
   &SolveForcesAndVelocity(Clamp Velocity=false, Limit Acceleration=false)
   &FN_UpdateAnimTrail(
    Locked Particle Color:LinearColor=&MultiplyLinearColor(
     Linear Color=$Particles.Initial.Color
     Scale Factor=&MakeFloat4FromVectorAndFloat("Vector (RGB)"=<1 1 1>, "Float (Alpha)"=$System.FN_Integrator001.Value)
    )
    Locked Particle RibbonWidth=1
    SkeletalMesh=$System.Skeletal Mesh
   ) {module=FN_UpdateAnimTrail001}
   &ScaleColor(
    Color Value To Scale=$Particles.Color
    Scale Alpha=&FN_FrontEndFloatModifier(Gameplay=1, Front End=0.35)
    Scale RGB=&Lerp_Vector(
     A=&MakeVectorFromLinearColorRGB(Color=&MultiplyLinearColor(
       Linear Color=&FNTimeOfDayColor(
        Night Color=<rgba 0.3515 0.4852 0.6445 0>
        Morning Color=<rgba 0.9911 0.7384 0.8167 0>
        Day Color=<rgba 1 1 1 1>
        Evening Color=<rgba 0.5972 0.3882 0.4192 0>
       )
       Scale Factor=<1 1 1 1>
      ))
     B=&MakeVectorFromLinearColorRGB(Color=<rgba 1 0.9967 0.99 1>)
     Alpha=&FN_FrontEndFloatModifier(Gameplay=0, Front End=1)
    )
    ScaleA=true
    ScaleRGB=true
   ) {enabled=false}
   &Color(
    Color=&MultiplyLinearColor(
     Linear Color=$System.VariantColor
     Scale Factor=&MakeVector4_FromLinearColor(LinearColor=&MultiplyLinearColor(
       Linear Color=&FNTimeOfDayColor(
        Night Color=<rgba 0.3515 0.4852 0.6445 1>
        Morning Color=<rgba 0.9911 0.7384 0.8167 1>
        Day Color=<rgba 1 1 1 1>
        Evening Color=<rgba 0.5972 0.3882 0.4192 1>
       )
       Scale Factor=&MakeVector4(
        X=1
        Y=1
        Z=1
        W=&FloatFromCurve(
         FloatCurve=DataInterface() {propertyValues {Curve {keys [<at 0.04 0>,<at 0.045 1>]}}}
         Scale Curve=&MakeCustomFloatFromBool(
          Bool=&SetBoolByIntComparison(A=&SpawnGroup(), B=2, Comparison Type=::NewEnumerator6)
          True Float=1
          False Float=0.6
         )
        )
       )
      ))
    )
    Scale Alpha=&Multiply_Float(
     A=&Multiply_Float(A=0.5, B=$Particles.AlphaFadeIn)
     B=&FN_FrontEndFloatModifier(Gameplay=$System.FN_Integrator001.Value, Front End=1)
    )
   )
   /SetVariables(Particles.RibbonU0Override=$Particles.NormalizedAge) {isSetParameters=true, module=SetVariables_#4}
  ]
  rendererClasses [&NiagaraRibbonRendererProperties]
  renderers [{rendererClass=&NiagaraRibbonRendererProperties}]
 }
 BlackDetailTrail {
  emitterSpawnScript [&FN_RibbonIDOffsetUtility()]
  emitterUpdateScript [
   &EmitterState(
    Inactive Response=::NewEnumerator1
    Life Cycle Mode=::NewEnumerator1
    Loop Behavior=::NewEnumerator1
    Loop Duration Mode=::NewEnumerator1
    UseLoopDelay=false
   )
   &FN_SpawnAnimTrail(Number of Anim Trails=3)
  ]
  particleSpawnScript [
   &FN_InitializeAnimTrailBoneIndecies()
   &FN_InitializeAnimTrailBoneIndecies(AnimTrail Selection=1, First Bone or Socket Index=2, Second Bone or Socket Index=3) {module=FN_InitializeAnimTrailBoneIndecies001}
   &FN_InitializeAnimTrailBoneIndecies(AnimTrail Selection=2, First Bone or Socket Index=8, Second Bone or Socket Index=9) {module=FN_InitializeAnimTrailBoneIndecies002}
   /SetVariables(Particles.AlphaFadeIn:NiagaraFloat=&Min_Float(
     A:NiagaraFloat=&Divide_Float(
      A=&Subtract_Float(A=$Emitter.Age, B=&FN_FrontEndFloatModifier(Gameplay=0, Front End=1.5))
      B=1
     )
     B:NiagaraFloat=1
    )) {isSetParameters=true, module=SetVariables_#3}
   &FN_InitializeAnimTrail(
    AnimTrailWidth=&FN_FrontEndFloatModifier(Gameplay=0.55, Front End=0.3)
    Color=<rgba 0 0 0 2>
    Lifetime=&MakeCustomFloatFromBool(Bool=$Particles.Locked, True Float=100, False Float=0.75)
    Position Offset=&FN_FrontEndVector3Modifier(Gameplay=<0 -12 0>, Front End=<0 -10 0>)
    SkeletalMesh=$System.Skeletal Mesh
   )
   &AddVelocity(Velocity=&Multiply_VectorByFloat(Vector=<0 400 0>, Float=&MakeCustomFloatFromBool(Bool=$User.bIsFrontEnd, True Float=1, False Float=1)))
   &DynamicMaterialParameters(
    Index 0 Param 1=1
    Index 0 Param 2=1
    Index 0 Param 3=&Add_Float(
     A=-3
     B=&SelectFloatFromArray(Direct Array Index=$Particles.SpawnGroup, Float Selection Array=DataInterface() {propertyValues {FloatData [0.1,0.3]}})
    )
    Index 0 Param 4=&SelectFloatFromArray(
     Direct Array Index=&MakeIntFromFloat(Float=$User.VariantID)
     Float Selection Array=DataInterface() {propertyValues {FloatData [0.6,0.55]}}
    )
    Module.Index0Param0Write=true
    Module.Index0Param1Write=true
    Module.Index0Param2Write=true
    Module.Index0Param3Write=true
   )
  ]
  particleUpdateScript [
   &SolveForcesAndVelocity(Clamp Velocity=false, Limit Acceleration=false)
   &FN_UpdateAnimTrail(
    Locked Particle Color:LinearColor=&MultiplyLinearColor(
     Linear Color=$Particles.Initial.Color
     Scale Factor=&MakeFloat4FromVectorAndFloat("Vector (RGB)"=<1 1 1>, "Float (Alpha)"=$System.FN_Integrator001.Value)
    )
    Locked Particle RibbonWidth=1
    SkeletalMesh=$System.Skeletal Mesh
   ) {module=FN_UpdateAnimTrail001}
   &ScaleColor(
    Color Value To Scale=$Particles.Color
    Scale Alpha=&FN_FrontEndFloatModifier(Gameplay=1, Front End=0.35)
    Scale RGB=&Lerp_Vector(
     A=&MakeVectorFromLinearColorRGB(Color=&MultiplyLinearColor(
       Linear Color=&FNTimeOfDayColor(
        Night Color=<rgba 0.3515 0.4852 0.6445 0>
        Morning Color=<rgba 0.9911 0.7384 0.8167 0>
        Day Color=<rgba 1 1 1 1>
        Evening Color=<rgba 0.5972 0.3882 0.4192 0>
       )
       Scale Factor=<1 1 1 1>
      ))
     B=&MakeVectorFromLinearColorRGB(Color=<rgba 1 0.9967 0.99 1>)
     Alpha=&FN_FrontEndFloatModifier(Gameplay=0, Front End=1)
    )
    ScaleA=true
    ScaleRGB=true
   ) {enabled=false}
   &Color(
    Color=&MultiplyLinearColor(
     Linear Color=<rgba 0 0 0 3>
     Scale Factor=&MakeVector4_FromLinearColor(LinearColor=&MultiplyLinearColor(
       Linear Color=&FNTimeOfDayColor(
        Night Color=<rgba 0.7969 0.7969 0.7969 1>
        Morning Color=<rgba 0.9911 0.7384 0.8167 1>
        Day Color=<rgba 1 1 1 1>
        Evening Color=<rgba 0.5972 0.3882 0.4192 1>
       )
       Scale Factor=&MakeVector4(
        X=1
        Y=1
        Z=1
        W=&FloatFromCurve(FloatCurve=DataInterface() {propertyValues {Curve {keys [<at 0 0>,<at 0.04 0>,<at 0.045 1>]}}})
       )
      ))
    )
    Scale Alpha=&Multiply_Float(
     A=&Multiply_Float(A=0.5, B=$Particles.AlphaFadeIn)
     B=&FN_FrontEndFloatModifier(Gameplay=$System.FN_Integrator001.Value, Front End=1)
    )
   )
   /SetVariables(Particles.RibbonU0Override=$Particles.NormalizedAge) {isSetParameters=true, module=SetVariables_#4}
   &DynamicMaterialParameters(
    Index 0 Param 1=&FloatFromCurve(FloatCurve=DataInterface() {propertyValues {Curve {keys [{interpMode=RCIM_Cubic, leaveTangent=-0.0514, tangentMode=RCTM_Break, time=0, value=0.9}, <at 1 0>]}}})
    Module.Index0Param0Write=true
    Module.Index0Param1Write=false
    Module.Index0Param2Write=false
    Module.Index0Param3Write=false
   ) {module=DynamicMaterialParameters001}
  ]
  rendererClasses [&NiagaraRibbonRendererProperties]
  renderers [
   {rendererClass=&NiagaraRibbonRendererProperties}
   {rendererClass=&NiagaraRibbonRendererProperties, rendererIndex=1}
  ]
 }
 GlitterTrails {
  emitterUpdateScript [
   &EmitterState(Scalability Mode=::NewEnumerator1)
   &SpawnRate(
    SpawnRate=&FN_FrontEndFloatModifier(Gameplay=&Lerp_Float(A=0, B=20, Alpha=$User.ForwardDot), Front End=0)
    Use Spawn Probability=false
   )
  ]
  particleSpawnScript [
   &InitializeParticle(
    Color=<rgba 5 1.8439 0 1>
    Lifetime=0.3
    Sprite Size Max=<130 400>
    Sprite Size Min=<80 300>
    Sprite Size Mode=::NewEnumerator2
    UsePositionOffset=false
    Write Material Random=true
   )
   &SkeletalMeshLocation(
    Bone / Socket Index=&SelectIntFromArray(
     Array Sampling Mode=::NewEnumerator1
     Direct Array Index=?
     Int Selection Array=DataInterface() {propertyValues {IntData [4,5,10]}}
     Random Seed=0
    )
    MaskSpawnGroup=false
    Mesh Sampling Type=::NewEnumerator5
    Skeletal Mesh=$System.Skeletal Mesh
    Socket Sampling Mode=::NewEnumerator5
   )
   &AddVelocity(Velocity=&FN_FrontEndVector3Modifier(Gameplay=&RandomRangeVector(Minimum=<0 450 0>, Maximum=<0 500 0>), Front End=<0 500 0>))
   &DynamicMaterialParameters(
    Index 0 Param 1=&RandomRangeFloat(Minimum=0, Maximum=1)
    Index 0 Param 2=&RandomRangeFloat(Minimum=0, Maximum=1)
    Index 0 Param 3=1
    Index 0 Param 4=&RandomRangeFloat(Minimum=0, Maximum=0)
    Module.Param0WriteEnabled=true
    Module.Param1WriteEnabled=true
    Module.Param2WriteEnabled=true
    Module.Param3WriteEnabled=true
   )
   &CylinderLocation(
    Cylinder Height=20
    Cylinder Radius=20
    MaskSpawnGroup=false
    Offset=<0 0 30>
    Offset Mode:ENiagara_OffsetMode=::NewEnumerator0
    Override Seed=false
    Surface Only=true
    UseOwnerScale=true
   ) {enabled=false}
   &SpriteFacingAndAlignment(Sprite Facing=<0 0 1>, WriteAlignment=false, WriteFacing=true)
  ]
  particleUpdateScript [
   &SolveForcesAndVelocity(Clamp Velocity=false, Limit Acceleration=false)
   &ParticleState()
   &ScaleColor(
    Scale Alpha=&FloatFromCurve(
     FloatCurve=DataInterface() {propertyValues {Curve {keys [<at 0 0>,<at 0.2 1>,<at 0.6 1>,<at 1 0>]}}}
     Scale Curve=&RandomRangeFloat(Minimum=30, Maximum=60)
    )
    ScaleA=true
    ScaleRGB=true
   )
   &ScaleSpriteSize(
    Non-Uniform Curve Sprite Scale=DataInterface() {propertyValues {
      XCurve {keys [{tangentMode=RCTM_User, time=0, value=1}, <at 0.7 1>, {tangentMode=RCTM_User, time=1, value=0.35}]}
      YCurve {keys [{tangentMode=RCTM_User, time=0, value=0.6}, <at 0.7 0.875>, {tangentMode=RCTM_User, time=1, value=0.35}]}
     }}
    Scale Sprite Size Mode=::NewEnumerator3
   )
  ]
  rendererClasses [&NiagaraSpriteRendererProperties]
  renderers [
   {rendererClass=&NiagaraSpriteRendererProperties}
   {rendererClass=&NiagaraSpriteRendererProperties, rendererIndex=1}
  ]
 }
 GoldenTrail {
  emitterSpawnScript [&FN_RibbonIDOffsetUtility()]
  emitterUpdateScript [
   &EmitterState(
    Inactive Response=::NewEnumerator1
    Life Cycle Mode=::NewEnumerator1
    Loop Behavior=::NewEnumerator1
    Loop Duration Mode=::NewEnumerator1
    UseLoopDelay=false
   )
   &FN_SpawnAnimTrail(Number of Anim Trails=3)
  ]
  particleSpawnScript [
   &FN_InitializeAnimTrailBoneIndecies()
   &FN_InitializeAnimTrailBoneIndecies(AnimTrail Selection=1, First Bone or Socket Index=2, Second Bone or Socket Index=3) {module=FN_InitializeAnimTrailBoneIndecies001}
   &FN_InitializeAnimTrailBoneIndecies(AnimTrail Selection=2, First Bone or Socket Index=6, Second Bone or Socket Index=7) {enabled=false, module=FN_InitializeAnimTrailBoneIndecies003}
   &FN_InitializeAnimTrailBoneIndecies(AnimTrail Selection=2, First Bone or Socket Index=8, Second Bone or Socket Index=9) {module=FN_InitializeAnimTrailBoneIndecies002}
   /SetVariables(Particles.AlphaFadeIn:NiagaraFloat=&Min_Float(
     A:NiagaraFloat=&Divide_Float(
      A=&Subtract_Float(A=$Emitter.Age, B=&FN_FrontEndFloatModifier(Gameplay=0, Front End=1.5))
      B=1
     )
     B:NiagaraFloat=1
    )) {isSetParameters=true, module=SetVariables_#3}
   &FN_InitializeAnimTrail(
    AnimTrailWidth=&FN_FrontEndFloatModifier(
     Gameplay=&MakeCustomFloatFromBool(
      Bool=&SetBoolByIntComparison(A=$Particles.SpawnGroup, B=2, Comparison Type=::NewEnumerator6)
      True Float=1.3
      False Float=1
     )
     Front End=0.8
    )
    Color=<rgba 0 0 0 2>
    Lifetime=&MakeCustomFloatFromBool(Bool=$Particles.Locked, True Float=100, False Float=0.75)
    Position Offset=&FN_FrontEndVector3Modifier(Gameplay=<0 -12 0>, Front End=<0 -10 0>)
    SkeletalMesh=$System.Skeletal Mesh
   )
   &AddVelocity(Velocity=&Multiply_VectorByFloat(Vector=<0 400 0>, Float=&MakeCustomFloatFromBool(Bool=$User.bIsFrontEnd, True Float=1, False Float=1)))
   &DynamicMaterialParameters(
    Index 0 Param 1=1
    Index 0 Param 2=2
    Index 0 Param 3=&Add_Float(
     A=-3
     B=&SelectFloatFromArray(Direct Array Index=$Particles.SpawnGroup, Float Selection Array=DataInterface() {propertyValues {FloatData [0.1,0.3]}})
    )
    Index 0 Param 4=1
    Module.Index0Param0Write=true
    Module.Index0Param1Write=true
    Module.Index0Param2Write=true
    Module.Index0Param3Write=true
   )
  ]
  particleUpdateScript [
   &SolveForcesAndVelocity(Clamp Velocity=false, Limit Acceleration=false)
   &FN_UpdateAnimTrail(
    Locked Particle Color:LinearColor=&MultiplyLinearColor(
     Linear Color=$Particles.Initial.Color
     Scale Factor=&MakeFloat4FromVectorAndFloat("Vector (RGB)"=<1 1 1>, "Float (Alpha)"=$System.FN_Integrator001.Value)
    )
    Locked Particle RibbonWidth=1
    SkeletalMesh=$System.Skeletal Mesh
   ) {module=FN_UpdateAnimTrail001}
   &ScaleColor(
    Color Value To Scale=$Particles.Color
    Scale Alpha=&FN_FrontEndFloatModifier(Gameplay=1, Front End=0.35)
    Scale RGB=&Lerp_Vector(
     A=&MakeVectorFromLinearColorRGB(Color=&MultiplyLinearColor(
       Linear Color=&FNTimeOfDayColor(
        Night Color=<rgba 0.3515 0.4852 0.6445 0>
        Morning Color=<rgba 0.9911 0.7384 0.8167 0>
        Day Color=<rgba 1 1 1 1>
        Evening Color=<rgba 0.5972 0.3882 0.4192 0>
       )
       Scale Factor=<1 1 1 1>
      ))
     B=&MakeVectorFromLinearColorRGB(Color=<rgba 1 0.9967 0.99 1>)
     Alpha=&FN_FrontEndFloatModifier(Gameplay=0, Front End=1)
    )
    ScaleA=true
    ScaleRGB=true
   ) {enabled=false}
   &Color(
    Color=&MultiplyLinearColor(
     Linear Color=<rgba 2 0.9036 0.2082 0.15>
     Scale Factor=&MakeVector4_FromLinearColor(LinearColor=&MultiplyLinearColor(
       Linear Color=&FNTimeOfDayColor(
        Night Color=<rgba 1.1 1.1 1.1 1>
        Morning Color=<rgba 0.9911 0.7384 0.8167 1>
        Day Color=<rgba 1 1 1 1>
        Evening Color=<rgba 0.5972 0.3882 0.4192 1>
       )
       Scale Factor=&MakeVector4(
        X=1
        Y=1
        Z=1
        W=&FloatFromCurve(FloatCurve=DataInterface() {propertyValues {Curve {keys [<at 0 0>,<at 0.04 0>,<at 0.045 1>]}}})
       )
      ))
    )
    Scale Alpha=&Multiply_Float(
     A=&Multiply_Float(A=0.5, B=$Particles.AlphaFadeIn)
     B=&FN_FrontEndFloatModifier(Gameplay=$System.FN_Integrator001.Value, Front End=1)
    )
   )
   /SetVariables(Particles.RibbonU0Override=$Particles.NormalizedAge) {isSetParameters=true, module=SetVariables_#4}
   &DynamicMaterialParameters(
    Index 0 Param 1=&FloatFromCurve(FloatCurve=DataInterface() {propertyValues {Curve {keys [{interpMode=RCIM_Cubic, leaveTangent=-0.0514, tangentMode=RCTM_Break, time=0, value=0.9}, <at 1 0>]}}})
    Module.Index0Param0Write=true
    Module.Index0Param1Write=false
    Module.Index0Param2Write=false
    Module.Index0Param3Write=false
   ) {module=DynamicMaterialParameters001}
  ]
  rendererClasses [&NiagaraRibbonRendererProperties]
  renderers [
   {rendererClass=&NiagaraRibbonRendererProperties}
   {rendererClass=&NiagaraRibbonRendererProperties, rendererIndex=1}
  ]
 }
 Sparckle_Flying {
  emitterUpdateScript [
   &EmitterState(
    Delay First Loop Only=true
    Life Cycle Mode=::NewEnumerator1
    Loop Delay=&FN_FrontEndFloatModifier(Gameplay=0, Front End=1.45)
    UseLoopDelay=true
   )
   &SpawnRate(
    SpawnRate=&FN_FrontEndFloatModifier(Gameplay=&Lerp_Float(A=0, B=25, Alpha=$User.ForwardDot), Front End=40)
    Use Spawn Probability=false
   )
  ]
  particleSpawnScript [
   &InitializeParticle(
    Color=<rgba 10 3.6476 0 1>
    Lifetime=&RandomRangeFloat(Minimum=0.7, Maximum=1)
    Sprite Rotation Mode=::NewEnumerator2
    Sprite Size Mode=::NewEnumerator3
    Uniform Sprite Size=&Multiply_Float(
     A=&FN_FrontEndFloatModifier(Gameplay=4.5, Front End=12)
     B=&FN_FrontEndFloatModifier(Gameplay=1, Front End=0.45)
    )
    UsePositionOffset=false
    Write Material Random=true
   )
   &SkeletalMeshLocation(
    Bone / Socket Index=&SelectIntFromArray(
     Array Sampling Mode=::NewEnumerator1
     Direct Array Index=?
     Int Selection Array=DataInterface() {propertyValues {IntData [4,5,10]}}
     Random Seed=0
    )
    MaskSpawnGroup=false
    Mesh Sampling Type=::NewEnumerator5
    Skeletal Mesh=$System.Skeletal Mesh
    Socket Sampling Mode=::NewEnumerator5
   )
   &CurlNoiseForce(Noise Frequency=4, Noise Strength=75, Pan Noise=true, Pan Noise Field=<1 1 0>) {enabled=false}
   &SphereLocation(
    Hemisphere Z=true
    MaskSpawnGroup=false
    Override Seed=false
    Sphere Radius=20
    Surface Only=false
    UseNonUniformScale=false
    UseOffset=false
    UseSphereOrientation=false
   ) {enabled=false}
   &BoxLocation(Box Size=<90 0 1>, MaskSpawnGroup=false, Module.Override Seed=false, Rotate Box=false, Surface Only=false)
   &ApplyInitialForces()
   &AddVelocity(Velocity=&FN_FrontEndVector3Modifier(
     Gameplay=&RandomRangeVector(Minimum=<0 800 0>, Maximum=<0 1000 0>)
     Front End=&RandomRangeVector(Minimum=<0 800 0>, Maximum=<0 1000 0>)
    ))
  ]
  particleUpdateScript [
   &ParticleState()
   &InheritSourceMovement(
    Applied Position Scale=&Multiply_VectorByFloat(
     Vector=<0.7 0.7 1>
     Float=&FloatFromCurve(FloatCurve=DataInterface() {propertyValues {Curve {keys [{interpMode=RCIM_Cubic, tangentMode=RCTM_User, time=0, value=1}, {interpMode=RCIM_Cubic, tangentMode=RCTM_User, time=1, value=0.3}]}}})
    )
    ApplyForce=false
    ApplyPosition=true
    ApplyPositionFalloff=false
   ) {enabled=false}
   &ScaleSpriteSize(Scale Factor=&Vector2DFromFloat(Value=&FloatFromCurve(FloatCurve=DataInterface() {propertyValues {Curve {keys [{interpMode=RCIM_Cubic, tangentMode=RCTM_User, time=0.25, value=1}, {interpMode=RCIM_Cubic, tangentMode=RCTM_User, time=1, value=0}]}}})))
   &AccelerationForce(
    Acceleration=&MakeVector(
     X=0
     Y=0
     Z=&MakeCustomFloatFromBool(
      Bool=$User.bIsFrontEnd
      True Float=&FloatFromCurve(FloatCurve=DataInterface() {propertyValues {Curve {keys [<at 0 -1>,<at 1 3>]}}}, Scale Curve=1500)
      False Float=0
     )
    )
    Coordinate Space=::Simulation
   )
   &Drag(Drag=1.5, Use Linear Drag=true, Use Rotational Drag=false)
   &SolveForcesAndVelocity(Clamp Velocity=false, Limit Acceleration=false)
   &ScaleColor(
    Scale Alpha=&FloatFromCurve(
     FloatCurve=DataInterface() {propertyValues {Curve {keys [{interpMode=RCIM_Cubic, tangentMode=RCTM_User, time=0, value=1}, {interpMode=RCIM_Cubic, tangentMode=RCTM_User, time=1, value=0}]}}}
     Scale Curve=&FN_FrontEndFloatModifier(Gameplay=$System.FN_Integrator001.Value, Front End=0.8)
    )
    ScaleA=true
    ScaleRGB=true
   )
  ]
  rendererClasses [&NiagaraSpriteRendererProperties]
  renderers [{rendererClass=&NiagaraSpriteRendererProperties}]
 }
 Speedlines {
  emitterUpdateScript [
   &EmitterState(
    Inactive Response=::NewEnumerator1
    Life Cycle Mode=::NewEnumerator1
    Loop Behavior=::NewEnumerator1
    Loop Duration Mode=::NewEnumerator1
    UseLoopDelay=false
   ) {module=EmitterState001}
   &SpawnRate(
    SpawnRate=&FN_FrontEndFloatModifier(
     Gameplay=&Multiply_Float(A=$User.ForwardDot, B=&UniformRangedFloat(Minimum=75, Maximum=150))
     Front End=0
    )
    Use Spawn Probability=true
   )
  ]
  particleSpawnScript [
   /Niagara/Modules/Spawn/Initialization/InitializeParticle(
    Color=&MakeLinearColorFromVectorAndFloat(
     "Vector (RGB)"=<1 1 1>
     "Float (Alpha)"=&Multiply_Float(
      A=&UniformRangedFloat(Minimum=0.03, Maximum=0.075)
      B=&Add_Float(
       A=1
       B=&Multiply_Float(A=10, B=&Subtract_Float(A=1, B=$Engine.Emitter.SpawnCountScale))
      )
     )
    )
    Lifetime=&UniformRangedFloat(Minimum=0.2, Maximum=0.35)
    Module.Write Color=true
    Module.Write Lifetime=true
    Module.Write Mass=true
    Module.Write Material Random=true
    Module.Write Position=false
    Module.Write Scale=false
    Module.Write SpriteRotation=false
    Module.Write SpriteSize=true
    Position=&SimulationPosition()
    Sprite Size=&UniformRangedVector2D(Minimum=<3 4>, Maximum=<5 6>)
   )
   &CylinderLocation(
    Cylinder Radius=&UniformRangedFloat(Minimum=100, Maximum=150)
    MaskSpawnGroup=false
    Module.Enable Lathe Profile=false
    Module.Override Randomness=false
    Module.Surface Only=true
    Orientation Axis:ENiagaraOrientationAxis=::XAxis
    Use Endcaps In Surface Only Mode=false
   )
   &AddVelocity(Velocity=<0 1000 0>)
   &AddVelocityFromPoint(Module.Use Falloff Distance=false, Velocity Strength=&UniformRangedFloat(Minimum=500, Maximum=700))
  ]
  particleUpdateScript [
   &ParticleState()
   &SolveForcesAndVelocity(Clamp Velocity=false, Limit Acceleration=false)
   &ScaleSpriteSize(Scale Factor=&Vector2DFromCurve(
     Vector2Curve=DataInterface() {propertyValues {XCurve {keys [<at 0 0.5 RCIM_Cubic>,<at 0.5 1 RCIM_Cubic>]}, YCurve {keys [<at 0 0 RCIM_Cubic>,<at 0.5 1 RCIM_Cubic>]}}}
     ScaleCurve=&Vector2DFromFloat(Value=&UniformRangedFloat(Minimum=0.5, Maximum=2))
    ))
   &ScaleSpriteSizeBySpeed(Max Scale Factor=<1 40>, Min Scale Factor=<0.01 0.05>, Velocity Threshold=1)
   &ScaleColor(
    Scale Alpha=&FloatFromCurve(FloatCurve=DataInterface() {propertyValues {Curve {keys [<at 0 0 RCIM_Cubic>, <at 0.2 1>, {arriveTangent=-2.221, interpMode=RCIM_Cubic, leaveTangent=-2.221, tangentMode=RCTM_User, time=1, value=0}]}}})
    ScaleA=true
    ScaleRGB=true
   )
  ]
  rendererClasses [&NiagaraSpriteRendererProperties]
  renderers [{rendererClass=&NiagaraSpriteRendererProperties}]
 }
}
summary {
 systemName=NS_Glider_GoldCat_Trail
 emitters [
  {emitterName=Speedlines, rendererClasses [&NiagaraSpriteRendererProperties]}
  {emitterName=BaseDarkTrail, rendererClasses [&NiagaraRibbonRendererProperties]}
  {emitterName=GoldenTrail, rendererClasses [&NiagaraRibbonRendererProperties]}
  {emitterName=Sparckle_Flying, rendererClasses [&NiagaraSpriteRendererProperties]}
  {emitterName=GlitterTrails, rendererClasses [&NiagaraSpriteRendererProperties]}
  {emitterName=BlackDetailTrail, rendererClasses [&NiagaraRibbonRendererProperties]}
 ]
 userVariables [
  User.bIsFrontEnd {type=NiagaraBool}
  User.bIsFrontEndPreview {type=NiagaraBool}
  User.bIsFullyDeployed {type=NiagaraBool}
  User.ForwardDot {type=NiagaraFloat, defaultValue=0, description=""}
  User.RibbonIDOffset {type=NiagaraInt32}
  User.RightDot {type=NiagaraFloat, defaultValue=0, description=""}
  User.SkeletalMesh {
   type=NiagaraDataInterfaceSkeletalMesh
   defaultValue {
    struct=NiagaraExt_VariableValue_DataInterface
    value {dataInterface=&self:NiagaraDataInterfaceSkeletalMesh_0}
   }
  }
  User.VariantID {type=NiagaraFloat, defaultValue=0, description=""}
 ]
}
